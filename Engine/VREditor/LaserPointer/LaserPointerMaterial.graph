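# Unlit material for the VR editor laser pointer on spline meshes, drawn without depth testing.
# LaserColor combines with a panning power texture and stepped sine crawl bands, both mapped by a sphere mask around the actor position.
# Animated world-space noise modulates brightness and opacity, while a tangent-space view falloff shades the edges.
# Eye adaptation inverse compensates for exposure, and feature-level switches give ES2/ES3.1 flat color and full opacity.
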Material /Engine/VREditor/LaserPointer/LaserPointerMaterial
graph {
 properties {
  bDisableDepthTest=true
  bEnableResponsiveAA=true
  bUseEmissiveForDynamicAreaLighting=true
  bUsedWithSplineMeshes=true
  pixelDepthOffsetMode=PDOM_Legacy
  shadingModel=MSM_Unlit
  usedShadingModels=Unlit
 }
 nodes [
  n1 = CameraVectorWS()
  Abs(None=ComponentMask(None=Transform(None=@n1) {transformSourceType=TRANSFORMSOURCE_World, transformType=TRANSFORM_Tangent}) {b=true})
  n2 = VectorParameter() {parameterName=LaserColor, defaultValue=<rgba 1 0 0 0>}
  Add(A=Constant() {r=0.03}, B=@n2.RGB)
  AppendVector()
  n3 = ComponentMask(None=TextureCoordinate() {vTiling=2}) {r=true}
  n4 = Multiply(
   A=SphereMask(A=ActorPositionWS(), B=WorldPosition().XYZ, Radius=ScalarParameter() {parameterName=Length}) {attenuationRadius=250, hardnessPercent=0}
   B=ScalarParameter() {defaultValue=1, parameterName=CrawlTiling}
  )
  AppendVector(A=@n3, B=@n4)
  n5 = Panner(Coordinate=AppendVector(A=@n3, B=Divide(A=@n4) {constB=50})) {speedX=0.01, speedY=0.3}
  n6 = Multiply(A=Multiply(B=TextureSample(UVs=@n5) {texture=&/VR_LaserPower_01}.RGB) {constA=16}, B=@n2.RGB)
  n7 = Time()
  n8 = ScalarParameter() {defaultValue=1, parameterName=Crawl Speed}
  n9 = Sine(None=Add(A=Multiply(A=@n7, B=@n8), B=@n4))
  n10 = Ceil(None=@n9)
  n11 = Add(A=Multiply(A=Clamp(None=@n10), B=ScalarParameter() {parameterName=Crawl}), B=@n2.RGB)
  n12 = ScreenPosition()
  n13 = Constant()
  n14 = Divide(A=Add(A=AppendVector(A=@n12.ViewportUV, B=@n13), B=WorldPosition().XYZ), B=ScalarParameter() {defaultValue=10, parameterName=Scale})
  n15 = Time()
  n16 = Add(A=ComponentMask(None=@n14) {r=true}, B=Multiply(A=@n15, B=ScalarParameter() {defaultValue=1, parameterName=X_Speed}))
  n17 = Add(A=ComponentMask(None=@n14) {g=true}, B=Multiply(A=@n15, B=ScalarParameter() {defaultValue=1, parameterName=Y_Speed}))
  n18 = ComponentMask(None=@n14) {b=true}
  n19 = Multiply(A=@n15, B=ScalarParameter() {defaultValue=1, parameterName=Z_Speed})
  n20 = Noise!(World Position=AppendVector(A=AppendVector(A=@n16, B=@n17), B=Add(A=@n18, B=@n19)), FilterWidth) {bTurbulence=false, levels=2, outputMin=0, scale=4}
  n21 = ScalarParameter() {defaultValue=10, parameterName=Amplify}
  n22 = Multiply(A=@n21, B=ScalarParameter() {defaultValue=0.11, parameterName=EdgeDarkener}) {constB=0.2}
  n23 = ComponentMask(None=Transform(None=CameraVectorWS()) {transformSourceType=TRANSFORMSOURCE_World, transformType=TRANSFORM_Tangent}) {r=true}
  n24 = Power(Base=OneMinus(None=Abs(None=@n23))) {constExponent=4}
  n25 = Multiply(A=Multiply(A=Add(A=@n6, B=@n11), B=@n20), B=Multiply(A=@n20, B=LinearInterpolate(A=@n22, B=@n21, Alpha=@n24) {constA=1, constB=5}))
  EyeAdaptationInverse(
   LightValueInput=FeatureLevelSwitch(Default=@n25, ES2=@n2.RGB, ES3_1=@n2.RGB, SM4_REMOVED=@n25, SM5=@n25, SM6=@n25)
   AlphaInput=Constant() {r=1}
  )
  n26 = Multiply(A=@n20, B=ScalarParameter() {defaultValue=1, parameterName=Opacity})
  n27 = Constant() {r=1}
  FeatureLevelSwitch(Default=@n26, ES2=@n27, ES3_1=@n27, SM4_REMOVED=@n26, SM5=@n26, SM6=@n26)
  FeatureLevelSwitch(Default=@n20, ES2=@n27, ES3_1=@n27, SM4_REMOVED=@n20, SM5=@n20, SM6=@n20)
  Floor(None=@n9)
 ]
}
