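# Unlit translucent material for the core of the VR Editor laser pointer, flagged for use on spline meshes.
# A glowing UserColor is shaped by inverted Fresnel, then lerped toward a dim constant by an animated pattern from two rotated, panning sine waves on the second UV channel.
# The color is compensated with EyeAdaptationInverse.
# An opacity term combines the inverted Fresnel, CameraDepthFade and an Opacity parameter.
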Material /Engine/VREditor/LaserPointer/M_LaserPointer-Core
graph {
 properties {bUsedWithSplineMeshes=true, blendMode=BLEND_Translucent, pixelDepthOffsetMode=PDOM_Legacy, shadingModel=MSM_Unlit, usedShadingModels=Unlit}
 nodes [
  n1 = Multiply(
   A=VectorParameter() {parameterName=UserColor, defaultValue=<rgba 1 0.1309 0.0012 0>}.RGB
   B=ScalarParameter() {defaultValue=10, parameterName=Glow}
  )
  n2 = OneMinus(None=Fresnel(
    ExponentIn=ScalarParameter() {defaultValue=0.5, parameterName=Exponent}
    BaseReflectFractionIn=ScalarParameter() {defaultValue=0.25, parameterName=BaseReflect}
    Normal=PixelNormalWS()
   ))
  n3 = TextureCoordinate() {coordinateIndex=1}
  n4 = Multiply(A=Time(), B=ScalarParameter() {defaultValue=0.5, parameterName=TimeMultiplier})
  n5 = Panner(Coordinate=Rotator(Coordinate=@n3, Time=Constant() {r=-0.2}) {speed=6.2832}, Time=@n4) {speedX=-0.25, speedY=-0.25}
  n6 = Panner(Coordinate=Rotator(Coordinate=@n3, Time=Constant() {r=0.25}) {speed=6.2832}, Time=@n4) {speedX=0.5}
  n7 = DotProduct(A=ComponentMask(None=Sine(None=@n5) {period=0.15}) {r=true}, B=ComponentMask(None=Sine(None=@n6) {period=0.125}) {g=true})
  EyeAdaptationInverse(LightValueInput=LinearInterpolate(A=Multiply(A=@n1, B=@n2), Alpha=Clamp(None=@n7)) {constB=0.25}, AlphaInput=Constant() {r=1})
  n8 = ScalarParameter() {defaultValue=10, parameterName=Glow}
  n9 = OneMinus(None=Fresnel(
    ExponentIn=ScalarParameter() {defaultValue=0.5, parameterName=Exponent}
    BaseReflectFractionIn=ScalarParameter() {defaultValue=0.25, parameterName=BaseReflect}
    Normal=PixelNormalWS()
   ))
  Multiply(A=LinearInterpolate(A=Multiply(B=@n8)) {constB=0.2}, B=@n9)
  Multiply(
   A=Clamp(
    None=Multiply(A=@n9) {constB=0.25}
    Min=ScalarParameter() {defaultValue=0.125, parameterName=Clamp-Min}
    Max=ScalarParameter() {defaultValue=1, parameterName=Clamp-Max}
   )
   B=VertexColor().R
  )
  n10 = ScalarParameter() {defaultValue=8, parameterName=FadeLenght}
  n11 = ScalarParameter() {defaultValue=4, parameterName=FadeOffset}
  Multiply(
   A=Multiply(
    A=@n2
    B=MaterialFunctionCall(Fade Length=@n10, Fade Offset=@n11) {materialFunction=&CameraDepthFade}.Result
   )
   B=ScalarParameter() {defaultValue=1, parameterName=Opacity, sliderMax=1}
  )
  ScalarParameter() {defaultValue=5, parameterName=BeamMultipli}
 ]
}
