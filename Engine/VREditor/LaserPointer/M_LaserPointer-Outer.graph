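# Translucent, unlit, two-sided material for the outer layer of the VR Editor laser pointer, flagged for use on spline meshes.
# A pixel-normal Fresnel term scales a tintable UserColor and Glow, with EyeAdaptationInverse countering auto exposure on the result.
# The same Fresnel, multiplied by a CameraDepthFade and an Opacity parameter, forms the alpha, so the beam fades near the camera.
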Material /Engine/VREditor/LaserPointer/M_LaserPointer-Outer
graph {
 properties {
  bUsedWithSplineMeshes=true
  blendMode=BLEND_Translucent
  pixelDepthOffsetMode=PDOM_Legacy
  shadingModel=MSM_Unlit
  twoSided=true
  usedShadingModels=Unlit
 }
 nodes [
  n1 = VectorParameter() {parameterName=UserColor, defaultValue=<rgba 1 0.1135 0.0038 0>}
  n2 = ScalarParameter() {defaultValue=1.25, parameterName=Glow}
  n3 = Fresnel(
   ExponentIn=ScalarParameter() {defaultValue=4, parameterName=Exponent}
   BaseReflectFractionIn=ScalarParameter() {defaultValue=0.05, parameterName=BaseReflect}
   Normal=PixelNormalWS()
  )
  EyeAdaptationInverse(LightValueInput=Multiply(A=Multiply(A=@n1.RGB, B=@n2), B=@n3), AlphaInput=Constant() {r=1})
  n4 = ScalarParameter() {defaultValue=64, parameterName=FadeLenght}
  n5 = ScalarParameter() {defaultValue=8, parameterName=FadeOffset}
  Multiply(
   A=Multiply(
    A=@n3
    B=MaterialFunctionCall(Fade Length=@n4, Fade Offset=@n5) {materialFunction=&CameraDepthFade}.Result
   )
   B=ScalarParameter() {defaultValue=0.5, parameterName=Opacity, sliderMax=1}
  )
 ]
}
